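; GiftBox header Blueprint interface (BPI_GiftBoxHeader) for the gift box feature.
; It declares a single function, GetCustomBackground, which returns a custom background value.
; The body holds only a default empty return, so any actual background comes from classes that implement the interface.
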(fn GetCustomBackground
  (return 0))
